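# Niagara system for the Eggplant vehicle boost cosmetic, emitting a fiery trail at the boost position plus car offset, following source movement.
# A flipbook emitter runs four spawn groups, with switches varying motion and size per group and VisibilityTag feeding four SubUV sprite renderers with separate color curves.
# A second emitter sprays small orange sparks backward, stirred by curl noise and vortex forces.
# Painted-variant override modules are present but disabled in both emitters.
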NiagaraSystem /VehicleCosmetics/Boosts/Boost_Eggplant/FX/NS_Boost_Eggplant_Revised
dependencies {
 usedDataInterfaces [
  &NiagaraDataInterfaceCurve
  &NiagaraDataInterfaceSpriteRendererInfo
  &NiagaraDataInterfaceVector2DCurve
  &NiagaraDataInterfaceColorCurve
  &NiagaraDataInterfaceSkeletalMesh
 ]
 usedDynamicInputs []
 usedModules [
  &self:SystemSpawnScript.SystemScriptSource.SystemScriptGraph.NiagaraNodeAssignment_0.SetVariables_#1
  &SystemState
 ]
}
emitters {
 NE_VC_SingleBoost_Flipbooks {
  emitterUpdateScript [
   &EmitterState()
   &SpawnRate(SpawnRate=&RandomRangeFloat(Minimum=4, Maximum=6), Use Spawn Probability=false)
   &SpawnRate(SpawnGroup=1, SpawnRate=&RandomRangeFloat(Minimum=20, Maximum=25), Use Spawn Probability=false) {module=SpawnRate001}
   &SpawnRate(SpawnGroup=2, SpawnRate=&RandomRangeFloat(Minimum=45, Maximum=55), Use Spawn Probability=false) {module=SpawnRate002}
   &SpawnRate(SpawnGroup=3, SpawnRate=&RandomRangeFloat(Minimum=10, Maximum=15), Use Spawn Probability=false) {module=SpawnRate003}
  ]
  particleSpawnScript [
   &VC_PaintedVariantsOverride() {enabled=false}
   /SetVariables(Particles.VisibilityTag=$Emitter.SpawnGroup) {isSetParameters=true}
   &InitializeParticle(
    Color Mode=::NewEnumerator0
    Lifetime=&MergeHelper_FloatSwitch(
     0=&RandomRangeFloat(Minimum=0.8, Maximum=1.1)
     1=&RandomRangeFloat(Minimum=1.2, Maximum=1.3)
     2=&RandomRangeFloat(Minimum=1.24, Maximum=1.26)
     3=&RandomRangeFloat(Minimum=1.2, Maximum=1.35)
     Index=$Emitter.SpawnGroup
     Num=::NewEnumerator2
    )
    Position Offset=&MergeHelper_VectorSwitch(
     0=&AddVectorToPosition(
      Position=<-20 0 0>
      Vector=&AddVectorToPosition(Position=$User.DTCarOffset, Vector=Linked() {linkedVariable {name=User.SingleBoostPos, type=NiagaraPosition}})
     )
     1=&AddVectorToPosition(
      Position=<-20 0 0>
      Vector=&AddVectorToPosition(Position=$User.DTCarOffset, Vector=Linked() {linkedVariable {name=User.SingleBoostPos, type=NiagaraPosition}})
     )
     2=&AddVectorToPosition(
      Position=<-25 0 0>
      Vector=&AddVectorToPosition(Position=$User.DTCarOffset, Vector=Linked() {linkedVariable {name=User.SingleBoostPos, type=NiagaraPosition}})
     )
     3=&AddVectorToPosition(
      Position=<-25 0 0>
      Vector=&AddVectorToPosition(Position=$User.DTCarOffset, Vector=Linked() {linkedVariable {name=User.SingleBoostPos, type=NiagaraPosition}})
     )
     Index=$Emitter.SpawnGroup
     Num=::NewEnumerator2
    )
    Position Offset Coordinate Space=::Local
    Sprite Rotation Angle=&RandomRangeFloat(Minimum=-120, Maximum=80)
    Sprite Rotation Mode=::NewEnumerator0
    Sprite Size Mode=::NewEnumerator1
    Uniform Sprite Size Max=&MergeHelper_FloatSwitch(0=200, 1=85, 2=100, 3=85, Index=$Emitter.SpawnGroup, Num=::NewEnumerator2)
    Uniform Sprite Size Min=&MergeHelper_FloatSwitch(0=150, 1=75, 2=50, 3=70, Index=$Emitter.SpawnGroup, Num=::NewEnumerator2)
    UsePositionOffset=true
    Write Material Random=true
   )
   &AddVelocity(Velocity=&MergeHelper_VectorSwitch(
     0=&RandomRangeVector(Minimum=<-100 0 20>, Maximum=<-150 1 0>)
     1=<-500 0 10>
     2=<-500 0 10>
     3=<-500 0 10>
     Index=$Emitter.SpawnGroup
     Num=::NewEnumerator2
    ))
  ]
  particleUpdateScript [
   &ParticleState()
   &SpriteRotationRate(Rotation Rate=&MergeHelper_FloatSwitch(
     0=&RandomRangeFloat(Minimum=-100, Maximum=100)
     1=&RandomRangeFloat(Minimum=-250, Maximum=250)
     2=&RandomRangeFloat(Minimum=-250, Maximum=250)
     3=&RandomRangeFloat(Minimum=-250, Maximum=250)
     Index=$Emitter.SpawnGroup
     Num=::NewEnumerator2
    )) {module=SpriteRotationRate001}
   &ScaleSpriteSize(Uniform Curve Sprite Scale=DataInterface() {propertyValues {Curve {
       defaultValue=1
       keys [
        {arriveTangent=1.6947, interpMode=RCIM_Cubic, leaveTangent=1.6947, tangentMode=RCTM_User, time=0, value=0.7}
        {arriveTangent=0.35, interpMode=RCIM_Cubic, leaveTangent=0.35, time=0.2, value=1.25}
        {arriveTangent=-0.0027, interpMode=RCIM_Cubic, leaveTangent=-0.0027, tangentMode=RCTM_User, time=1, value=1.05}
       ]
      }}})
   &CurlNoiseForce(
    Bias Noise=false
    Noise Frequency=150
    Noise Strength=&MergeHelper_FloatSwitch(
     0=&RandomRangeFloat(Minimum=700, Maximum=800)
     1=0
     2=0
     3=0
     Index=$Emitter.SpawnGroup
     Num=::NewEnumerator2
    )
    Pan Noise=false
   )
   &InheritSourceMovement(ApplyForce=false, ApplyPosition=true, ApplyPositionFalloff=false)
   &SolveForcesAndVelocity(Clamp Velocity=false, Limit Acceleration=false)
   &SubUVAnimation(
    End Frame Range Override=32
    "Sprite Renderer "=DataInterface() {propertyValues {SpriteRenderer=&self:NE_VC_SingleBoost_Flipbooks_0.NiagaraSpriteRendererProperties_0}}
    Start Frame Range Override=21
    UseEndFrame=true
    UseStartFrame=true
   )
   &SubUVAnimation(
    End Frame Range Override=64
    "Sprite Renderer "=DataInterface() {propertyValues {SpriteRenderer=&self:NE_VC_SingleBoost_Flipbooks_0.NiagaraSpriteRendererProperties_1}}
    Start Frame Range Override=30
    UseEndFrame=true
    UseStartFrame=true
   ) {module=SubUVAnimation001}
   &SubUVAnimation(
    "Sprite Renderer "=DataInterface() {propertyValues {SpriteRenderer=&self:NE_VC_SingleBoost_Flipbooks_0.NiagaraSpriteRendererProperties_4}}
    Start Frame Range Override=30
    UseEndFrame=true
    UseStartFrame=true
   ) {module=SubUVAnimation002}
   &SubUVAnimation(
    "Sprite Renderer "=DataInterface() {propertyValues {SpriteRenderer=&self:NE_VC_SingleBoost_Flipbooks_0.NiagaraSpriteRendererProperties_5}}
    Start Frame Range Override=30
    UseEndFrame=true
    UseStartFrame=true
   ) {module=SubUVAnimation003}
   /SetVariables(
    Particles.Color001:LinearColor=&RandomRangeLinearColor(
     Minimum=&ColorFromCurve(ColorCurve=DataInterface() {propertyValues {
        AlphaCurve {keys [
          <at 0.2856 0>
          <at 0.3595 0.2>
          {arriveTangent=2.4481, interpMode=RCIM_Cubic, leaveTangent=2.4481, time=0.4242, value=0.5}
          <at 0.4821 0.5>
          <at 0.7704 0 RCIM_Cubic>
         ]}
        BlueCurve {keys [<at 0.264 0.2865>,<at 0.3909 0.1>,<at 0.4583 0.014>,<at 0.5431 0>]}
        GreenCurve {keys [<at 0.264 1.4648>,<at 0.3909 0.2>,<at 0.4583 0.028>,<at 0.5431 0>]}
        RedCurve {keys [<at 0.264 5>,<at 0.3909 0.5>,<at 0.4583 0.07>,<at 0.5431 0>]}
       }})
     Maximum=&ColorFromCurve(ColorCurve=DataInterface() {propertyValues {
        AlphaCurve {keys [<at 0.2902 0>,<at 0.6019 1 RCIM_Cubic>,<at 0.7731 0.3855>,<at 0.8831 0 RCIM_Cubic>]}
        BlueCurve {keys [<at 0.2651 0.01>,<at 0.5037 0>]}
        GreenCurve {keys [<at 0.2651 0.01>,<at 0.5037 0>]}
        RedCurve {keys [<at 0.2651 0.01>,<at 0.5037 0>]}
       }})
    )
    Particles.Color002=&ColorFromCurve(ColorCurve=DataInterface() {propertyValues {
       AlphaCurve {keys [<at 0.0219 0>,<at 0.367 1>,<at 0.7194 0 RCIM_Cubic>]}
       BlueCurve {keys [<at 0.2062 0.0156>,<at 0.2772 0.0022>,<at 0.418 0.0021>,<at 0.6586 0>]}
       GreenCurve {keys [<at 0.2062 0.0156>,<at 0.2772 0.0515>,<at 0.418 0.0474>,<at 0.6586 0>]}
       RedCurve {keys [<at 0.2062 0.0156>,<at 0.2772 0.15>,<at 0.418 0.1382>,<at 0.6586 0>]}
      }})
    Particles.Color_1:LinearColor=&ColorFromCurve(ColorCurve=DataInterface() {propertyValues {
       AlphaCurve {keys [<at 0.1134 1>,<at 0.5896 1>]}
       BlueCurve {keys [<at 0.3035 0.4688>,<at 0.3607 0>,<at 0.5152 0>,<at 0.601 0>,<at 0.6506 0>]}
       GreenCurve {keys [<at 0.3035 4.1016>,<at 0.3607 4.1667>,<at 0.5152 0.75>,<at 0.601 0.75>,<at 0.6506 0>]}
       RedCurve {keys [<at 0.3035 15>,<at 0.3607 25>,<at 0.5152 15>,<at 0.601 15>,<at 0.6506 0>]}
      }})
    Particles.Dynamic_0:Vector4f=<1 0 0 1>
    Particles.Size_0:Vector2f=&Multiply_Vector2D_ByFloat(
     Float=95
     Vector2D=&Vector2DFromCurve(
      Vector2Curve=DataInterface() {propertyValues {
        XCurve {
         defaultValue=1
         keys [
          {arriveTangent=1.1605, interpMode=RCIM_Cubic, leaveTangent=1.1605, tangentMode=RCTM_User, time=0, value=0.8}
          {arriveTangent=-0.6, interpMode=RCIM_Cubic, leaveTangent=-0.6, time=0.65, value=1.1}
          {arriveTangent=-4.279, interpMode=RCIM_Cubic, leaveTangent=-4.279, tangentMode=RCTM_User, time=1, value=0.2}
         ]
        }
        YCurve {
         defaultValue=1
         keys [
          {arriveTangent=1.1605, interpMode=RCIM_Cubic, leaveTangent=1.1605, tangentMode=RCTM_User, time=0, value=0.8}
          {arriveTangent=-0.6, interpMode=RCIM_Cubic, leaveTangent=-0.6, time=0.65, value=1.1}
          {arriveTangent=-4.6901, interpMode=RCIM_Cubic, leaveTangent=-4.6901, tangentMode=RCTM_User, time=1, value=0.2}
         ]
        }
       }}
      ScaleCurve=<1 1>
     )
    )
   ) {isSetParameters=true}
   &Color(Color=&ColorFromCurve(ColorCurve=DataInterface() {propertyValues {
       AlphaCurve {keys [<at 0 1>,<at 0.788 0>]}
       BlueCurve {keys [<at 0.0254 0.4187>, {arriveTangent=-0.543, interpMode=RCIM_Cubic, leaveTangent=-0.543, time=0.3015, value=0.2}, <at 0.7965 0>]}
       GreenCurve {keys [<at 0.0254 1.4729>, {arriveTangent=-1.757, interpMode=RCIM_Cubic, leaveTangent=-1.757, time=0.3015, value=1.67}, <at 0.7965 0.118>]}
       RedCurve {keys [<at 0.0254 2>, {arriveTangent=3.8904, interpMode=RCIM_Cubic, leaveTangent=3.8904, time=0.3015, value=10}, <at 0.7965 5>]}
      }}))
   &DynamicMaterialParameters(
    Index 1 Param 1=&FloatFromCurve(FloatCurve=DataInterface() {propertyValues {Curve {
        defaultValue=1
        keys [
         {arriveTangent=-0.0225, interpMode=RCIM_Cubic, leaveTangent=-0.0225, tangentMode=RCTM_User, time=0, value=-0.5}
         {arriveTangent=2.4118, interpMode=RCIM_Cubic, leaveTangent=2.4118, time=0.76, value=0.9}
         {interpMode=RCIM_Cubic, tangentMode=RCTM_User, time=0.85, value=1.55}
        ]
       }}})
    Module.Index1Float0Write=true
    Module.Index1Float1Write=true
    Module.Index1Float2Write=true
    Module.Index1Float3Write=true
    Write Parameter Index 0=::NewEnumerator0
    Write Parameter Index 1=::NewEnumerator1
   ) {module=DynamicMaterialParameters001}
  ]
  rendererClasses [&NiagaraSpriteRendererProperties]
  renderers [
   {rendererClass=&NiagaraSpriteRendererProperties}
   {rendererClass=&NiagaraSpriteRendererProperties, rendererIndex=1}
   {rendererClass=&NiagaraSpriteRendererProperties, rendererIndex=2}
   {rendererClass=&NiagaraSpriteRendererProperties, rendererIndex=3}
  ]
 }
 NE_VC_SingleBoost_Sprites {
  emitterUpdateScript [
   &EmitterState()
   &SpawnRate(SpawnRate=&RandomRangeFloat(Minimum=33, Maximum=38), Use Spawn Probability=false)
  ]
  particleSpawnScript [
   &VC_PaintedVariantsOverride() {enabled=false}
   &InitializeParticle(
    Color Mode=::NewEnumerator0
    Lifetime=&RandomRangeFloat(Minimum=0.5, Maximum=0.55)
    Position Offset=&AddVectorToPosition(
     Position=<-40 0 0>
     Vector=&AddVectorToPosition(Position=$User.DTCarOffset, Vector=Linked() {linkedVariable {name=User.SingleBoostPos, type=NiagaraPosition}})
    )
    Position Offset Coordinate Space=::Local
    Sprite Rotation Mode=::NewEnumerator2
    Sprite Size Mode=::NewEnumerator1
    Uniform Sprite Size Max=4
    Uniform Sprite Size Min=3.5
    UsePositionOffset=true
    Write Material Random=true
   )
   &AddVelocity(
    Cone Angle=30
    Module.UseConeFalloff=false
    Module.UseVelDistribution=false
    Override Seed=false
    Velocity Mode=::NewEnumerator2
    Velocity Speed=&RandomRangeFloat(Minimum=-700, Maximum=-800)
   )
   &ShapeLocation(
    MaskSpawnGroup=false
    Override Seed=false
    Ring Radius=55
    Rotation Mode=::NewEnumerator1
    Shape Primitive=::NewEnumerator4
    U Distribution=0.5
    UseOwnerScale=true
   ) {module=ShapeLocation001}
  ]
  particleUpdateScript [
   &ParticleState()
   &ScaleSpriteSize(Uniform Curve Sprite Scale=DataInterface() {propertyValues {Curve {
       defaultValue=0
       keys [{interpMode=RCIM_Cubic, tangentMode=RCTM_User, time=0.7, value=1}, {interpMode=RCIM_Cubic, tangentMode=RCTM_User, time=0.8, value=0}]
      }}})
   &CurlNoiseForce(
    Bias Noise=false
    Noise Frequency=450
    Noise Strength=&RandomRangeFloat(Minimum=250, Maximum=350)
    Pan Noise=false
   )
   &VortexForce(
    UseOriginOffset=false
    Vortex Axis=<0 1 0>
    Vortex Force Amount=&RandomRangeFloat(Minimum=350, Maximum=450)
   )
   &InheritSourceMovement(ApplyForce=false, ApplyPosition=true, ApplyPositionFalloff=false)
   &SolveForcesAndVelocity(Clamp Velocity=false, Limit Acceleration=false)
   &Color(Color=&ColorFromCurve(ColorCurve=DataInterface() {propertyValues {
       AlphaCurve {keys [<at 0.7813 1 RCIM_Cubic>,<at 1 0 RCIM_Cubic>]}
       BlueCurve {keys [<at 0.3321 0 RCIM_Cubic>,<at 0.447 0>]}
       GreenCurve {keys [<at 0.3321 2.0549 RCIM_Cubic>,<at 0.447 0.3333>]}
       RedCurve {keys [<at 0.3321 5 RCIM_Cubic>,<at 0.447 5>]}
      }}))
   /SetVariables(
    Particles.Color1:LinearColor=&ColorFromCurve(ColorCurve=DataInterface() {propertyValues {
       AlphaCurve {keys [<at 0.7181 1 RCIM_Cubic>,<at 1 0 RCIM_Cubic>]}
       BlueCurve {keys [<at 0.4298 0 RCIM_Cubic>,<at 0.7426 0>]}
       GreenCurve {keys [<at 0.4298 2.0549 RCIM_Cubic>,<at 0.7426 0.1126>]}
       RedCurve {keys [<at 0.4298 5 RCIM_Cubic>,<at 0.7426 5>]}
      }})
    Particles.Size1:Vector2f=&Multiply_Vector2D_ByFloat(
     Float=30
     Vector2D=&Vector2DFromCurve(
      Vector2Curve=DataInterface() {propertyValues {
        XCurve {
         defaultValue=1
         keys [{interpMode=RCIM_Cubic, tangentMode=RCTM_User, time=0.5, value=1}, {interpMode=RCIM_Cubic, tangentMode=RCTM_User, time=0.9, value=0}]
        }
        YCurve {
         defaultValue=1
         keys [{interpMode=RCIM_Cubic, tangentMode=RCTM_User, time=0.5, value=1}, {interpMode=RCIM_Cubic, tangentMode=RCTM_User, time=0.9, value=0}]
        }
       }}
      ScaleCurve=<1 1>
     )
    )
   ) {isSetParameters=true}
  ]
  rendererClasses [&NiagaraSpriteRendererProperties]
  renderers [
   {rendererClass=&NiagaraSpriteRendererProperties}
   {rendererClass=&NiagaraSpriteRendererProperties, rendererIndex=1}
  ]
 }
}
summary {
 systemName=NS_Boost_Eggplant_Revised
 emitters [
  {emitterName=NE_VC_SingleBoost_Flipbooks, rendererClasses [&NiagaraSpriteRendererProperties]}
  {emitterName=NE_VC_SingleBoost_Sprites, rendererClasses [&NiagaraSpriteRendererProperties]}
 ]
 userVariables [
  User.bIsFrontendPreview {type=NiagaraBool}
  User.bIsPainted {type=NiagaraBool}
  User.ColorVariantID {type=NiagaraInt32}
  User.DTCarOffset {type=NiagaraPosition, description="", defaultValue {}}
  User.PaintedColor {type=LinearColor, description="", defaultValue {struct=LinearColor, value=<rgba 1 1 1 1>}}
  User.PaintedScalar {type=NiagaraFloat, defaultValue=1, description=""}
  User.PortCount {type=NiagaraInt32, defaultValue=1}
  User.SingleBoostPos {type=NiagaraPosition, description="", defaultValue {}}
 ]
}
